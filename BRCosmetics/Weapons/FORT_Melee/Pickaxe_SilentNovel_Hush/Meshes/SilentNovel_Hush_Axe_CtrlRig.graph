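# Control Rig for the SilentNovel Hush pickaxe that procedurally drives its piston bones from the hammer bone.
# On forward solve, a parent constraint makes Piston follow Hammer on Z translation and scale only, ignoring its rotation.
# An aim constraint then points Piston_Rot at Melee_OneHanded_Hammer_Ref, a child of Hammer, so the linkage tracks the hammer's motion.
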ControlRigBlueprint /BRCosmetics/Weapons/FORT_Melee/Pickaxe_SilentNovel_Hush/Meshes/SilentNovel_Hush_Axe_CtrlRig
graphs [
 RigVMFunctionLibrary {links [], nodes []}
 RigVMModel {nodes [
   RigUnit_BeginExecution = &RigUnit_BeginExecution()
   &RigUnit_ParentConstraint(
    AdvancedSettings=(InterpolationType=Average,RotationOrderForFilter=XZY)
    Child=(Type=Bone,Name="Piston")
    Filter=(TranslationFilter=(bX=false,bY=false,bZ=True),RotationFilter=(bX=false,bY=false,bZ=false),ScaleFilter=(bX=True,bY=True,bZ=True))
    Parents=((Item=(Type=Bone,Name="Hammer"),Weight=1))
    Weight=1
    ExecutePin=@RigUnit_BeginExecution.ExecutePin
   )
   > &RigUnit_AimConstraintLocalSpaceOffset(
    Child=(Type=Bone,Name="Piston_Rot")
    Filter=(bX=True,bY=True,bZ=True)
    Parents=((Item=(Type=Bone,Name="Melee_OneHanded_Hammer_Ref"),Weight=1))
    Weight=1
    WorldUp=(Target=(X=0,Y=0,Z=1),Kind=Location,Space=(Type=None,Name="None"))
   )
   &RigUnit_PrepareForExecution()
  ]}
]
hierarchy [
 b_AR_Root {loc=<0 0 0>, rot=<0 0 0 1>}
 Melee_TwoHanded_Handle {parent=b_AR_Root, loc=<0 0 0>, rot=<0 0 0.7071 -0.7071>}
 Hammer {parent=Melee_TwoHanded_Handle, loc=<-0.1042 -0.4714 19.3996>, rot=<-0.5 0.5 -0.5 -0.5>}
 Melee_OneHanded_Hammer_Ref {parent=Hammer, loc=<2.8497 0 -0.2254>, rot=<0 -0.7071 0 -0.7071>}
 Piston {parent=Melee_TwoHanded_Handle, loc=<-0.1042 3.121 18.2092>, rot=<0 0 -0.7071 -0.7071>}
 Piston_Rot {parent=Piston, loc=<0 0 2.9895>, rot=<0 -1 0 0>}
]
