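# Physics asset for the SilentNovel Hush two-handed melee pickaxe, giving its axe mesh simple collision and a jointed head.
# It has two capsule bodies, one on the Hammer bone and one on the Melee_TwoHanded_Handle bone.
# A single constraint links them, with very tight swing 1 and twist angles and a wider swing 2 angle, suggesting hinge-like movement in one plane.
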PhysicsAsset /BRCosmetics/Weapons/FORT_Melee/Pickaxe_SilentNovel_Hush/Meshes/SilentNovel_Hush_Axe_Physics
bodies {
 Hammer [{shapeType=Capsule, length=3, radius=11, shapeName=Hammer_capsule, center=<5 0 0>, rotation=<rot 90 0 0>}]
 Melee_TwoHanded_Handle [{
   shapeType=Capsule
   length=16.9104
   radius=4.0081
   shapeName=Melee_TwoHanded_Handle_capsule
   center=<-0.1042 0.6916 3.5676>
   rotation=<rot -0.2313 -90 0>
  }]
}
constraints [{bone1Name=Hammer, bone2Name=Melee_TwoHanded_Handle, swing1LimitDegrees=1, swing2LimitDegrees=25, twistLimitDegrees=1}]
